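# Additive unlit Niagara sprite material for an environmental aura effect on the Whisper Remedy Sweet shrine.
# Two wave-flame texture layers pan at different speeds and tilings in radial UVs, then get a contrast boost.
# Each layer takes its own tint and their sum blends between them, feeding shared emissive adjustments.
# The same sum, masked to a soft ring by two subtracted radial gradients, drives depth-faded opacity scaled by particle alpha.
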Material /Sprites/FX/Niagara/Shrine_WhisperRemedySweet/M_WhiserRemedySweet_Env_Aura
graph {
 properties {
  bAutomaticallySetUsageInEditor=false
  bUsedWithNiagaraSprites=true
  blendMode=BLEND_Additive
  shadingModel=MSM_Unlit
  usedShadingModels=Unlit
 }
 nodes [
  n1 = MaterialFunctionCall(Vector or UVs=TextureCoordinate() {vTiling=0.2}) {materialFunction=&VectorToRadialValue}
  n2 = MaterialFunctionCall(
   In=TextureSample(UVs=Panner(Coordinate=@n1.Radial Coordinates) {speedY=-0.075}) {texture=/Game/Effects/Fort_Effects/Textures/Tile/T_Tile_WaveFlame}.R
   Contrast=Constant() {r=0.1}
  ) {materialFunction=&CheapContrast}
  n3 = MaterialFunctionCall(Vector or UVs=TextureCoordinate() {vTiling=0.05}) {materialFunction=&VectorToRadialValue}
  n4 = MaterialFunctionCall(
   In=TextureSample(UVs=Panner(Coordinate=@n3.Radial Coordinates) {speedY=-0.05}) {texture=&T_Tile_WaveFlame}.RGB
   Contrast=Constant() {r=0.1}
  ) {materialFunction=&CheapContrast}
  n5 = Add(A=@n2.Result, B=@n4.Result)
  n6 = TextureCoordinate()
  n7 = Subtract(
   A=MaterialFunctionCall(UVs=@n6, Density=Constant() {r=2}) {materialFunction=&RadialGradientExponential}.RadialGradientExponential
   B=MaterialFunctionCall(UVs=@n6, Radius=Constant() {r=0.25}) {materialFunction=&RadialGradientExponential}.RadialGradientExponential
  )
  Multiply(
   A=ParticleColor().A
   B=MaterialFunctionCall!(
    InAlpha=Multiply(A=@n5, B=@n7)
    OpacityBasedFade_DistanceA=ScalarParameter() {defaultValue=200, group=Opacity, parameterName=OpacityBasedFade_DistanceA}
    OpacityBasedFade_DistanceB=ScalarParameter() {defaultValue=50, group=Opacity, parameterName=OpacityBasedFade_DistanceB}
   ) {materialFunction=/Game/Effects/Niagara/MasterMaterials/Functions/MF_FX_OpacityBasedDepthFade}.FinalAlpha
  )
  n8 = Multiply(A=VectorParameter() {group=Emissive, parameterName=EmissiveColorTintA, defaultValue=<rgba 1 0 0 1>}.RGB, B=@n2.Result)
  n9 = Multiply(A=VectorParameter() {group=Emissive, parameterName=EmissiveColorTintB, defaultValue=<rgba 0 0 1 1>}.RGB, B=@n4.Result)
  Multiply(B=MaterialFunctionCall(Emissive=LinearInterpolate(A=@n8, B=@n9, Alpha=@n5)) {materialFunction=&MF_FX_EmissiveAdjustments}.Result) {constA=1}
 ]
}
